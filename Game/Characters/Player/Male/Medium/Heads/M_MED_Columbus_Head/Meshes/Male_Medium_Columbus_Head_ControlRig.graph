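# Control Rig for the Columbus medium male player head mesh that reverses the head bone's twist about its local Z axis.
# It splits the head's local rotation into swing and twist, negates the twist angle, then multiplies it with the unchanged swing.
# The result is written back to the head and propagated to children, with debug axes drawn at the head before and after.
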ControlRigBlueprint /Game/Characters/Player/Male/Medium/Heads/M_MED_Columbus_Head/Meshes/Male_Medium_Columbus_Head_ControlRig
graphs [
 RigVMFunctionLibrary {links [], nodes []}
 RigVMModel {nodes [
   n1 = &RigUnit_BeginExecution()
   /Script/ControlRig.RigUnit_DebugTransformMutable(
    Color=(R=1,G=0,B=0,A=1)
    DebugDrawSettings=(DepthPriority=SDPG_Foreground,Lifetime=-1)
    Mode=Axes
    Scale=10
    Space=None
    Thickness=0
    WorldOffset=(Rotation=(X=0,Y=0,Z=0,W=1),Translation=(X=0,Y=0,Z=0),Scale3D=(X=1,Y=1,Z=1))
    bEnabled=True
    ExecutePin=@n1.ExecutePin
    Transform=&RigUnit_GetBoneTransform(Bone=head).Transform
   ) {is=struct RigVMUnitNode}
   n2 = &RigVMFunction_MathQuaternionSwingTwist(TwistAxis=(X=0,Y=0,Z=1), Input=&RigUnit_GetBoneTransform(Bone=head, Space=LocalSpace).Transform.Rotation)
   n3 = &RigVMFunction_MathFloatMul(
    B=-1
    A=/Script/RigVM.RigVMFunction_MathQuaternionToAxisAndAngle(Value=@n2.Twist) {is=struct RigVMUnitNode, template="ToAxisAndAngle::Execute(in Value,out Axis,out Angle)"}.Angle
   ) {template="Multiply::Execute(in A,in B,out Result)"}
   > /Script/ControlRig.RigUnit_SetBoneRotation(
    Bone=head
    Space=LocalSpace
    Weight=1
    bPropagateToChildren=True
    Rotation=/Script/RigVM.RigVMFunction_MathQuaternionMul(
     A=/Script/RigVM.RigVMFunction_MathQuaternionFromAxisAndAngle(Axis=(X=0,Y=0,Z=1), Angle=@n3.Result) {is=struct RigVMUnitNode, template="FromAxisAndAngle::Execute(in Axis,in Angle,out Result)"}.Result
     B=@n2.Swing
    ) {is=struct RigVMUnitNode, template="Multiply::Execute(in A,in B,out Result)"}.Result
   ) {is=struct RigVMUnitNode}
   > &RigUnit_DebugTransformMutable(
    Color=(R=1,G=0,B=0,A=1)
    DebugDrawSettings=(DepthPriority=SDPG_Foreground,Lifetime=-1)
    Mode=Axes
    Scale=10
    Space=None
    Thickness=0.3
    WorldOffset=(Rotation=(X=0,Y=0,Z=0,W=1),Translation=(X=0,Y=0,Z=0),Scale3D=(X=1,Y=1,Z=1))
    bEnabled=True
    Transform=&RigUnit_GetBoneTransform(Bone=head).Transform
   )
  ]}
]
hierarchy [
 root {loc=<0 0 0>, rot=<0 0 0 1>}
 pelvis {parent=root, loc=<0 -0.7071 93.1191>, rot=<0.0127 0.707 -0.0127 -0.707>}
 spine_01 {parent=pelvis, loc=<6.7494 -0.6388 0>, rot=<0 0 0.034 -0.9994>}
 spine_02 {parent=spine_01, loc=<7.2518 -0.035 0>, rot=<0 0 0.0238 -0.9997>}
 spine_03 {parent=spine_02, loc=<6.0104 0 0>, rot=<0 0 0.0149 -0.9999>}
 spine_04 {parent=spine_03, loc=<8.1024 0 0>, rot=<0 0 -0.0132 -0.9999>}
 spine_05 {parent=spine_04, loc=<10.0146 0 0>, rot=<0 0 -0.0015 -1>}
 neck_01 {parent=spine_05, loc=<11.3138 0.1529 0>, rot=<0 0 -0.2003 -0.9797>}
 neck_02 {parent=neck_01, loc=<4.5627 0 0>, rot=<0 0 0.0504 -0.9987>}
 head {parent=neck_02, loc=<5.3914 0 0>, rot=<0 0 0.115 -0.9934>}
 dyn_helmet {parent=head, loc=<8.605 2.341 0>, rot=<0 0.7071 0.7071 0>}
]
