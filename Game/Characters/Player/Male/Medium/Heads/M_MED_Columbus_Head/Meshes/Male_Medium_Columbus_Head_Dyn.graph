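# Physics asset for the dynamic mesh of the male medium Columbus player head, covering a helmet bone and the head bone.
# Each of the two bones has one small, identically sized capsule body, with the helmet capsule offset from its bone origin.
# A single constraint joins dyn_helmet to head with very tight swing limit angles and a slightly wider twist limit angle.
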PhysicsAsset /Game/Characters/Player/Male/Medium/Heads/M_MED_Columbus_Head/Meshes/Male_Medium_Columbus_Head_Dyn
bodies {
 dyn_helmet [{shapeType=Capsule, length=5, radius=3, center=<3.8114 0 0>, rotation=<rot 90 0 0>}]
 head [{shapeType=Capsule, length=5, radius=3, rotation=<rot 90 0 0>}]
}
constraints [{bone1Name=dyn_helmet, bone2Name=head, swing1LimitDegrees=1, swing2LimitDegrees=1, twistLimitDegrees=5}]
