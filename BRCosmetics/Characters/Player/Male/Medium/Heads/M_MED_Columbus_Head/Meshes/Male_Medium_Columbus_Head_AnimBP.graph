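# Animation Blueprint for the male medium Columbus cosmetic head, which copies its pose from the attached parent mesh.
# The AnimGraph constrains head_alt to head around bone edits that apply HeadRotation and a small yaw offset.
# Control Rig and RigidBody nodes sit unconnected to the output.
# The update event flags phone-booth exit and guinea-pig emote montages on the owner mesh, though the shown AnimGraph reads neither flag.
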AnimBlueprint /BRCosmetics/Characters/Player/Male/Medium/Heads/M_MED_Columbus_Head/Meshes/Male_Medium_Columbus_Head_AnimBP
target_skeleton=/Game/Characters/Player/Male/Medium/Heads/M_MED_Columbus_Head/Meshes/Male_Medium_Columbus_Head_Skeleton
graphs [
 AnimGraph {nodes [
   Control Rig() {
    alphaBoolBlend {blendOption=Linear}
    controlRigAssetReference {blueprintRigClass=/Game/Characters/Player/Male/Medium/Heads/M_MED_Columbus_Head/Meshes/Male_Medium_Columbus_Head_ControlRig.Male_Medium_Columbus_Head_ControlRig_C}
   }
   RigidBody() {
    bClampLinearTranslationLimitToRefPose=true
    bOverrideWorldGravity=true
    overridePhysicsAsset=/Game/Characters/Player/Male/Medium/Heads/M_MED_Columbus_Head/Meshes/Male_Medium_Columbus_Head_Dyn
    simulationSpace=BaseBoneSpace
    alphaBoolBlend {blendOption=Linear}
    baseBoneRef {boneName=head}
    componentAppliedLinearAccClamp=<3000 3000 3000>
    componentLinearAccScale=<0.4 0.3 0.3>
    componentLinearVelScale=<0.4 0.3 0.3>
    overrideWorldGravity {z=-980}
   }
   n1 = "Constraint - head_alt (P)"(ComponentPose=Local To Component(LocalPose=Copy Pose From Mesh() {bUseAttachedParent=true}), Alpha=1, ConstraintWeights_0=1) {
    alphaBoolBlend {blendOption=Linear}
    alphaScaleBiasClamp {bInterpResult=true, interpSpeedDecreasing=4}
    boneToModify {boneName=head_alt}
    constraintSetup [{targetBone {boneName=head}}]
    constraintWeights [1]
   }
   n2 = "Transform (Modify) Bone - Bone: head_alt"(ComponentPose=@n1, Alpha=1, Translation="0,0,0", Rotation=Get HeadRotation() -> (HeadRotation:Rotator), Scale="1,1,1") {
    rotationMode=BMM_Replace
    rotationSpace=BCS_BoneSpace
    translationSpace=BCS_BoneSpace
    alphaBoolBlend {blendOption=Linear}
    alphaScaleBiasClamp {interpSpeedDecreasing=1, interpSpeedIncreasing=1}
    boneToModify {boneName=head_alt}
   }
   n3 = "Constraint - head_alt (P)"(ComponentPose=@n2, Alpha=1, ConstraintWeights_0=1) {
    alphaBoolBlend {blendOption=Linear}
    alphaScaleBiasClamp {bInterpResult=true}
    boneToModify {boneName=head_alt}
    constraintSetup [{targetBone {boneName=head}}]
    constraintWeights [1]
   }
   n4 = "Transform (Modify) Bone - Bone: head_alt"!(ComponentPose=@n3, Alpha=1, Rotation="0,-5,0") -> (Pose) {
    rotationMode=BMM_Additive
    rotationSpace=BCS_BoneSpace
    alphaBoolBlend {blendOption=Linear}
    alphaScaleBiasClamp {bInterpResult=true, interpSpeedDecreasing=1, interpSpeedIncreasing=1}
    boneToModify {boneName=head_alt}
    rotation {yaw=-5}
   }
   Output Pose(Result=Component To Local(ComponentPose=@n4)) {layerGroup=None}
  ]}
 EventGraph {nodes [
   Event BlueprintUpdateAnimation()
   > n1 = Sequence()
   n2 = GetAnimInstance(self=GetOwnerSkeletalMeshComponent())
   > Set Helmet Constraint(Helmet Constraint:float="To Float (Boolean)"(InBool=Montage_IsPlaying(self=@n2, Montage=/Game/Animation/Game/MainPlayer/Reacts/WorldInteraction/Hideable/PhoneBooth_Exit_Montage_V2))) -> (Output_Get:float)
   n3 = GetAnimInstance(self=GetOwnerSkeletalMeshComponent())
   Is Valid(exec=@n1.then_1, InputObject=@n3)
   > Set isGuineaPigEmote(isGuineaPigEmote:bool=Montage_IsPlaying(self=@n3, Montage=/BRCosmetics/Animation/Game/MainPlayer/Emotes/GuineaPig/Emote_GuineaPig_CMM_FE_M)) -> (Output_Get:bool)
  ]}
]
variables [
 Helmet Constraint {type=number}
 isGuineaPigEmote {type=boolean}
 isNotPlayingEmote {type=boolean}
]
